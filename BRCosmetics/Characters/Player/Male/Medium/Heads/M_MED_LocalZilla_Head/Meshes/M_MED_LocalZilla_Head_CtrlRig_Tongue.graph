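# Control Rig for the LocalZilla player head that procedurally animates the jaw, neck flubs and tongue.
# A looping sine, gated and scaled by a seeded random value, is smoothed to open the jaw, with neck flubs following through a second interpolation.
# A jaw threshold sets jawOpen, which extends the tongue bones and weights a faster sine wiggle, offset cumulatively down the chain to the split tip.
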ControlRigBlueprint /BRCosmetics/Characters/Player/Male/Medium/Heads/M_MED_LocalZilla_Head/Meshes/M_MED_LocalZilla_Head_CtrlRig_Tongue
graphs [
 RigVMFunctionLibrary {links [], nodes []}
 RigVMModel {nodes [
   RigUnit_BeginExecution = &RigUnit_BeginExecution()
   n1 = /Script/RigVM.RigVMFunction_TimeLoop(Duration=1, Speed=2.5) {is=struct RigVMUnitNode}
   n2 = RigVMVariableNode(
    ExecuteContext=@RigUnit_BeginExecution.ExecutePin
    Value=/Script/RigVM.RigVMFunction_MathDoubleSin(Value=&RigVMFunction_MathDoubleMul(B=2, A=@n1.Absolute).Result) {is=struct RigVMUnitNode, template="Sin::Execute(in Value,out Result)"}.Result
   ) {variable=jawSinFunction}
   SequenceAggregate_1_1 = &RigVMFunction_Sequence(ExecuteContext=@n2.ExecuteContext)
   n3 = /Script/RigVM.RigVMFunction_RandomFloat(Duration=1, Maximum=0.5, Minimum=-1, Seed=217) {is=struct RigVMUnitNode, template="Random::Execute(in Seed,in Minimum,in Maximum,in Duration,out Result)"}
   Branch_1_RigVMUnitNode = &RigVMFunction_ControlFlowBranch(
    Condition=/Script/RigVM.RigVMFunction_MathDoubleLessEqual(B=0, A=@n3.Result) {is=struct RigVMUnitNode, template="LessEqual::Execute(in A,in B,out Result)"}.Result
    ExecuteContext=@SequenceAggregate_1_1.A
   )
   RigVMVariableNode(Value=&RigVMFunction_MathDoubleAdd(B=0.8, A=@n3.Result).Result, ExecuteContext=@Branch_1_RigVMUnitNode.False) {variable=alphaRandomizer}
   RigVMVariableNode(ExecuteContext=@Branch_1_RigVMUnitNode.True) {variable=alphaRandomizer}
   RigVMVariableNode(
    Value=&RigVMFunction_MathDoubleMul(B=RigVMVariableNode() {variable=alphaRandomizer}.Value, A=RigVMVariableNode() {variable=jawSinFunction}.Value).Result
    ExecuteContext=@SequenceAggregate_1_1.B
   ) {variable=jawSinFunction}
   n4 = &RigVMFunction_AlphaInterp(
    InterpSpeedDecreasing=10
    InterpSpeedIncreasing=10
    OutRange=(Max=1)
    Scale=1
    bClampResult=true
    bMapRange=true
    Bias=~
    ClampMin=~
    Value=&RigVMFunction_MathDoubleMul(B=1.5, A=RigVMVariableNode() {variable=jawSinFunction}.Value).Result
   )
   > n5 = &RigUnit_ModifyTransforms(
    ItemToModify=((Item=(Type=Bone,Name="C_jaw"),Transform=(Rotation=(X=0,Y=0,Z=0,W=1),Translation=(X=0,Y=0,Z=0),Scale3D=(X=1,Y=1,Z=1))))
    Weight=@n4.Result
    ItemToModify.0.Transform.Rotation=&RigVMFunction_MathQuaternionFromEuler(Euler=(X=-10,Y=0,Z=0)).Result
   )
   SequenceAggregate = &RigVMFunction_Sequence(ExecuteContext=@n5.ExecutePin)
   n6 = &RigVMFunction_AlphaInterp(
    InterpSpeedDecreasing=50
    InterpSpeedIncreasing=20
    OutRange=(Max=1)
    Scale=1
    bClampResult=true
    bMapRange=true
    Bias=~
    ClampMin=~
    Value=@n4.Result
   )
   &RigUnit_ModifyTransforms(
    ItemToModify="((Item=(Type=Bone,Name=\"dyn_L_neckFlub\"),Transform=(Rotation=(X=0,Y=0,Z=0,W=1),Translation=(X=0.2,Y=-0.1,Z=-0.2),Scale3D=(X=1,Y=1,Z=1))),(Item=(Type=Bone,Name=\"dyn_R_neckFlub\"),Transform=(Rotation=(X=0,Y=0,Z=0,W=1),Translation=(X=0.2,Y=-0.1,Z=0.2),Scale3D= ...(truncated)"
    Weight=@n6.Result
    ExecutePin=@SequenceAggregate.B
   )
   n7 = &RigVMFunction_AlphaInterp(
    InterpSpeedDecreasing=60
    InterpSpeedIncreasing=60
    OutRange=(Max=1)
    Scale=1
    bClampResult=true
    bMapRange=true
    Bias=~
    ClampMin=~
    Value=RigVMVariableNode() {variable=jawOpen}.Value
   )
   > &RigUnit_ModifyTransforms(
    ItemToModify=((Item=(Type=Bone,Name="dyn_tongue_1"),Transform=(Rotation=(X=0,Y=0,Z=0,W=1),Translation=(X=3,Y=0,Z=0),Scale3D=(X=1.2,Y=1,Z=1))))
    Weight=@n7.Result
    ItemToModify.0.Transform.Rotation=&RigVMFunction_MathQuaternionFromEuler(Euler=(X=0,Y=0,Z=-13)).Result
   )
   > &RigUnit_ModifyTransforms(
    ItemToModify=((Item=(Type=Bone,Name="dyn_tongue_2"),Transform=(Rotation=(X=0,Y=0,Z=0,W=1),Translation=(X=3,Y=0,Z=0),Scale3D=(X=1,Y=1,Z=1))))
    Weight=@n7.Result
   )
   > &RigUnit_ModifyTransforms(
    ItemToModify=((Item=(Type=Bone,Name="dyn_tongue_3"),Transform=(Rotation=(X=0,Y=0,Z=0,W=1),Translation=(X=2.5,Y=0,Z=0),Scale3D=(X=1,Y=1,Z=1))))
    Weight=@n7.Result
   )
   > &RigUnit_ModifyTransforms(
    ItemToModify=((Item=(Type=Bone,Name="dyn_tongue_4"),Transform=(Rotation=(X=0,Y=0,Z=0,W=1),Translation=(X=2,Y=0,Z=0),Scale3D=(X=1,Y=1,Z=1))))
    Weight=@n7.Result
   )
   Subtract = &RigVMFunction_MathDoubleSub(B=0.7, A=RigVMVariableNode() {variable=jawOpen}.Value)
   Divide = &RigVMFunction_MathDoubleDiv(B=5, A=@Subtract.Result)
   n8 = &RigVMFunction_MathDoubleMul(B=4, A=&RigVMFunction_TimeLoop(Duration=1, Speed=6).Absolute)
   Multiply = &RigVMFunction_MathDoubleMul(
    B=50
    A=&RigVMFunction_MathDoubleAdd(B=30, A=&RigVMFunction_MathDoubleSin(Value=@n8.Result) {template="Sin::Execute(in Value,out Result)"}.Result).Result
   )
   > &RigUnit_ModifyTransforms(
    ItemToModify=((Item=(Type=Bone,Name="dyn_tongue_2"),Transform=(Rotation=(X=0,Y=0,Z=0,W=1),Translation=(X=0,Y=0,Z=0),Scale3D=(X=1,Y=1,Z=1))))
    Weight=@Divide.Result
    ItemToModify.0.Transform.Rotation=&RigVMFunction_MathQuaternionFromEuler(Euler=(X=0,Y=-60,Z=0), Euler.Y=@Multiply.Result, Euler.Z=@Multiply.Result).Result
   )
   Add = &RigVMFunction_MathDoubleAdd(B=3, A=@Multiply.Result)
   > &RigUnit_ModifyTransforms(
    ItemToModify=((Item=(Type=Bone,Name="dyn_tongue_3"),Transform=(Rotation=(X=0,Y=0,Z=0,W=1),Translation=(X=0,Y=0,Z=0),Scale3D=(X=1,Y=1,Z=1))))
    ItemToModify.0.Transform.Rotation=&RigVMFunction_MathQuaternionFromEuler(Euler=(X=0,Y=40,Z=0), Euler.Y=@Add.Result, Euler.Z=@Add.Result).Result
    Weight=@Subtract.Result
   )
   n9 = &RigVMFunction_MathDoubleMul(B=1.4, A=&RigVMFunction_MathDoubleAdd(B=3, A=@Add.Result).Result)
   > &RigUnit_ModifyTransforms(
    ItemToModify=((Item=(Type=Bone,Name="dyn_tongue_4"),Transform=(Rotation=(X=0,Y=0,Z=0,W=1),Translation=(X=0,Y=0,Z=0),Scale3D=(X=1,Y=1,Z=1))))
    ItemToModify.0.Transform.Rotation=&RigVMFunction_MathQuaternionFromEuler(Euler=(X=0,Y=40,Z=0), Euler.Y=@n9.Result, Euler.Z=@n9.Result).Result
    Weight=@Subtract.Result
   )
   n10 = &RigVMFunction_MathDoubleAdd(B=3, A=@n9.Result)
   > &RigUnit_ModifyTransforms(
    ItemToModify=((Item=(Type=Bone,Name="dyn_tongueTip_L"),Transform=(Rotation=(X=0,Y=0,Z=0,W=1),Translation=(X=0,Y=0,Z=0),Scale3D=(X=1,Y=1,Z=1))))
    ItemToModify.0.Transform.Rotation=&RigVMFunction_MathQuaternionFromEuler(Euler=(X=0,Y=40,Z=0), Euler.Y=@n10.Result, Euler.Z=@n10.Result).Result
    Weight=@Subtract.Result
   )
   n11 = &RigVMFunction_MathDoubleAdd(B=2, A=@n10.Result)
   > &RigUnit_ModifyTransforms(
    ItemToModify=((Item=(Type=Bone,Name="dyn_tongueTip_R"),Transform=(Rotation=(X=0,Y=0,Z=0,W=1),Translation=(X=0,Y=0,Z=0),Scale3D=(X=1,Y=1,Z=1))))
    ItemToModify.0.Transform.Rotation=&RigVMFunction_MathQuaternionFromEuler(Euler=(X=0,Y=40,Z=0), Euler.Y=@n11.Result, Euler.Z=@n11.Result).Result
    Weight=@Subtract.Result
   )
   LessEqual = &RigVMFunction_MathDoubleLessEqual(B=0.25, A=@n4.Result) {template="LessEqual::Execute(in A,in B,out Result)"}
   Branch_RigVMUnitNode = &RigVMFunction_ControlFlowBranch(Condition=@LessEqual.Result, ExecuteContext=@SequenceAggregate.A)
   RigVMVariableNode(Value=1, ExecuteContext=@Branch_RigVMUnitNode.False) {variable=jawOpen}
   RigVMVariableNode(Value=0, ExecuteContext=@Branch_RigVMUnitNode.True) {variable=jawOpen}
   &RigVMFunction_MathDoubleMul(B=0.3, A=@n7.Result)
  ]}
]
hierarchy [
 root {loc=<0 0 0>, rot=<0 0 0 1>}
 pelvis {parent=root, loc=<0 -0.7071 93.1191>, rot=<0.0127 0.707 -0.0127 -0.707>}
 spine_01 {parent=pelvis, loc=<6.7494 -0.6388 0>, rot=<0 0 0.034 -0.9994>}
 spine_02 {parent=spine_01, loc=<7.2518 -0.035 0>, rot=<0 0 0.0238 -0.9997>}
 spine_03 {parent=spine_02, loc=<6.0104 0 0>, rot=<0 0 0.0149 -0.9999>}
 spine_04 {parent=spine_03, loc=<8.1024 0 0>, rot=<0 0 -0.0132 -0.9999>}
 spine_05 {parent=spine_04, loc=<10.0146 0 0>, rot=<0 0 -0.0015 -1>}
 neck_01 {parent=spine_05, loc=<11.3138 0.1529 0>, rot=<0 0 -0.2003 -0.9797>}
 neck_02 {parent=neck_01, loc=<4.5627 0 0>, rot=<0 0 0.0504 -0.9987>}
 head {parent=neck_02, loc=<5.3914 0 0>, rot=<0 0 0.115 -0.9934>}
 faceAttach {parent=head, loc=<8.0216 0.7066 0>, rot=<0.0029 -0.7071 -0.0029 -0.7071>}
 C_brow_mid {parent=faceAttach, loc=<0 9.1063 6.5601>, rot=<0.2859 0 0 -0.9582>}
 C_jaw {parent=faceAttach, loc=<-0.0134 0.3237 -0.7921>, rot=<0 0 0 1>}
 C_lip_lower_mid {parent=C_jaw, loc=<0.0223 16.5824 0.7902>, rot=<-0.1057 0 0 -0.9944>}
 L_lip_lower_outer {parent=C_jaw, loc=<4.0166 11.633 -1.0654>, rot=<-0.0913 0.1562 -0.1305 -0.9748>}
 R_lip_lower_outer {parent=C_jaw, loc=<-3.9898 11.633 -1.0654>, rot=<-0.0913 -0.1562 0.1305 -0.9748>}
 tongue {parent=C_jaw, loc=<0.0141 3.0868 0>, rot=<-0.1568 0 0 -0.9876>}
 dyn_tongue_1 {parent=tongue, loc=<-0.0096 1.4403 -0.8617>, rot=<-0.5526 -0.4412 0.5526 0.4412>}
 dyn_tongue_1_5 {parent=dyn_tongue_1, loc=<1.3081 0 0>, rot=<0 0 -0.0091 -1>}
 dyn_tongue_2 {parent=dyn_tongue_1_5, loc=<1.2663 0 0>, rot=<0 0 -0.0015 -1>}
 dyn_tongue_2_5 {parent=dyn_tongue_2, loc=<1.2445 0 0>, rot=<0 0 0.006 -1>}
 dyn_tongue_3 {parent=dyn_tongue_2_5, loc=<1.1941 0 0>, rot=<0.0018 0.0002 0.0038 -1>}
 dyn_tongue_3_5 {parent=dyn_tongue_3, loc=<1.1736 0 0>, rot=<0.0115 0.0011 0.0068 -0.9999>}
 dyn_tongue_4 {parent=dyn_tongue_3_5, loc=<1.0191 0 0>, rot=<0 0 0 1>}
 dyn_tongueTip_R {parent=dyn_tongue_4, loc=<1.0137 -0.0019 0.3118>, rot=<-0.0134 -0.0007 -0.0518 -0.9986>}
 dyn_tongueTip_L {parent=dyn_tongue_4, loc=<1.012 0.0146 -0.308>, rot=<-0.0134 -0.0007 -0.0518 -0.9986>}
 C_lip_upper_mid {parent=faceAttach, loc=<-0.0028 17.0164 1.1129>, rot=<0.2225 0.0013 -0.0005 -0.9749>}
 C_nose_bridge {parent=faceAttach, loc=<0 12.5593 4.5005>, rot=<-0.5064 0 0 -0.8623>}
 L_brow_mid {parent=faceAttach, loc=<4.8803 7.5267 5.498>, rot=<0.3788 0.0692 0.5248 -0.7591>}
 L_brow_outer {parent=faceAttach, loc=<6.199 4.8904 6.4743>, rot=<0.2684 -0.0874 0.6066 -0.7432>}
 L_cheek_inner {parent=faceAttach, loc=<6.4884 4.5493 1.5644>, rot=<0.0116 0.15 0.5717 -0.8065>}
 L_eye {parent=faceAttach, loc=<4.0613 6.2625 4.016>, rot=<-0.0059 -0.0165 -0.3342 -0.9423>}
 L_eye_lid_lower_mid {parent=L_eye, loc=<0.9642 -0.048 -0.729>, rot=<0.0329 0 0 -0.9995>}
 L_eye_lid_upper_mid {parent=L_eye, loc=<1.1171 -0.0474 0.6822>, rot=<0.0638 0.0126 0.0348 -0.9973>}
 L_lip_corner {parent=faceAttach, loc=<5.8997 4.3272 -0.8822>, rot=<0.0043 0.0074 -0.1876 -0.9822>}
 L_lip_upper_outer {parent=faceAttach, loc=<4.2289 11.882 -1.3818>, rot=<0.1434 -0.1162 0.5746 -0.7974>}
 R_brow_mid {parent=faceAttach, loc=<-4.8803 7.5267 5.498>, rot=<0.3788 -0.0692 -0.5248 -0.7591>}
 R_brow_outer {parent=faceAttach, loc=<-6.199 4.8904 6.4743>, rot=<0.2684 0.0874 -0.6066 -0.7432>}
 R_cheek_inner {parent=faceAttach, loc=<-6.4884 4.5493 1.5644>, rot=<0.0116 -0.15 -0.5717 -0.8065>}
 R_eye {parent=faceAttach, loc=<-4.0613 6.2625 4.0159>, rot=<-0.0059 0.0165 0.3342 -0.9423>}
 R_eye_lid_lower_mid {parent=R_eye, loc=<-0.9642 -0.048 -0.729>, rot=<0.0329 0 0 -0.9995>}
 R_eye_lid_upper_mid {parent=R_eye, loc=<-1.1171 -0.0473 0.6822>, rot=<0.0638 -0.0126 -0.0348 -0.9973>}
 R_lip_corner {parent=faceAttach, loc=<-5.8997 4.3272 -0.8822>, rot=<0.0043 -0.0074 0.1876 -0.9822>}
 R_lip_upper_outer {parent=faceAttach, loc=<-4.2289 11.882 -1.3818>, rot=<0.1434 0.1162 -0.5746 -0.7974>}
 dyn_R_bandana_1 {parent=faceAttach, loc=<-2.0808 -7.8417 6.0958>, rot=<0.2483 -0.8347 0.1401 -0.4712>}
 dyn_R_bandana_2 {parent=dyn_R_bandana_1, loc=<3.9925 0 0>, rot=<0 0 0 1>}
 dyn_L_bandana_1 {parent=faceAttach, loc=<1.2773 -7.7056 5.9659>, rot=<0.0973 -0.4114 0.2087 -0.8819>}
 dyn_L_bandana_2 {parent=dyn_L_bandana_1, loc=<4.2577 0 0>, rot=<0 0 0 1>}
 dyn_headband_front {parent=faceAttach, loc=<0.0501 7.1521 9.9043>, rot=<0 0 0 1>}
 dyn_R_neckFlub {parent=neck_02, loc=<4.6148 1.4194 6.2738>, rot=<0.0921 0.6965 -0.2772 -0.6554>}
 dyn_L_neckFlub {parent=neck_02, loc=<4.6148 1.4194 -6.2738>, rot=<-0.0921 -0.6965 -0.2772 -0.6554>}
]
variables [
 alphaRandomizer {type=float, default=0}
 jawOpen {type=double, default=0}
 jawSinFunction {type=double, default=0}
 jawSinFunctionOverlap {type=double, default=0}
]
