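# Physics asset for the male medium LocalZilla head cosmetic, giving capsule bodies to the neck, head and two dangling bandana tails.
# Large capsules cover neck_01, neck_02 and head, chained by constraints with free swing and twist.
# Left and right two-bone dyn_ bandana chains hang from the head on small capsules, with free swing but a tighter twist limit than the neck joints.
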PhysicsAsset /BRCosmetics/Characters/Player/Male/Medium/Heads/M_MED_LocalZilla_Head/Meshes/M_MED_LocalZilla_Head_Physics
bodies {
 dyn_L_bandana_1 [{shapeType=Capsule, length=4, radius=1, shapeName=dyn_L_bandana_1_capsule, center=<2 0 0>, rotation=<rot 90 0 0>}]
 dyn_L_bandana_2 [{shapeType=Capsule, length=5, radius=2, shapeName=dyn_L_bandana_2_capsule, center=<3.5 1 0>, rotation=<rot 90 176.4237 160.9966>}]
 dyn_R_bandana_1 [{shapeType=Capsule, length=4, radius=1, shapeName=dyn_R_bandana_1_capsule, center=<2 0 0>, rotation=<rot 90 0 0>}]
 dyn_R_bandana_2 [{shapeType=Capsule, length=5, radius=2, shapeName=dyn_R_bandana_2_capsule, center=<3.5 1 0>, rotation=<rot 90 172.875 157.448>}]
 head [{shapeType=Capsule, length=10, radius=8.5, shapeName=head_capsule, center=<8.0329 5.4214 0>, rotation=<rot 90 -165.9638 -80.9635>}]
 neck_01 [{shapeType=Capsule, length=10, radius=8.5, shapeName=neck_01_capsule, center=<4 -0.9016 0>, rotation=<rot 90 0 0>}]
 neck_02 [{shapeType=Capsule, length=10, radius=8.5, shapeName=neck_02_capsule, center=<4 -0.5813 0>, rotation=<rot 90 0 0>}]
}
constraints [
 {
  bone1Name=neck_02
  bone2Name=neck_01
  swing1LimitDegrees=45
  swing1Motion=Free
  swing2LimitDegrees=45
  swing2Motion=Free
  twistLimitDegrees=45
  twistMotion=Free
 }
 {
  bone1Name=head
  bone2Name=neck_02
  swing1LimitDegrees=45
  swing1Motion=Free
  swing2LimitDegrees=45
  swing2Motion=Free
  twistLimitDegrees=45
  twistMotion=Free
 }
 {
  bone1Name=dyn_L_bandana_1
  bone2Name=head
  swing1LimitDegrees=45
  swing1Motion=Free
  swing2LimitDegrees=45
  swing2Motion=Free
  twistLimitDegrees=20
 }
 {
  bone1Name=dyn_L_bandana_2
  bone2Name=dyn_L_bandana_1
  swing1LimitDegrees=45
  swing1Motion=Free
  swing2LimitDegrees=45
  swing2Motion=Free
  twistLimitDegrees=20
 }
 {
  bone1Name=dyn_R_bandana_1
  bone2Name=head
  swing1LimitDegrees=45
  swing1Motion=Free
  swing2LimitDegrees=45
  swing2Motion=Free
  twistLimitDegrees=20
 }
 {
  bone1Name=dyn_R_bandana_2
  bone2Name=dyn_R_bandana_1
  swing1LimitDegrees=45
  swing1Motion=Free
  swing2LimitDegrees=45
  swing2Motion=Free
  twistLimitDegrees=20
 }
]
