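# Animation Blueprint for the LocalZilla character head, adding tongue rigging, headband physics and material-driven cloth motion.
# The mesh-copied pose runs through a tongue Control Rig, then AnimDynamics and RigidBody headband simulation enabled by the Stage2 variant tag.
# A front-end idle curve blends DefaultSlot animation onto facial bones.
# The event graph feeds skydiving-dependent wind, velocity limits and bone motion into base and overlay materials for WPO cloth simulation.
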AnimBlueprint /BRCosmetics/Characters/Player/Male/Medium/Heads/M_MED_LocalZilla_Head/Meshes/M_MED_LocalZilla_Head_Skeleton_AnimBlueprint_customTongue
target_skeleton=&/M_MED_LocalZilla_Head_Skeleton
graphs [
 AnimGraph {nodes [
   n1 = Layered blend per bone(
    BasePose=Use cached pose 'Final Pose'()
    BlendPoses_0=Slot 'DefaultSlot'()
    BlendWeights_0=Get Custom Idle is Playing() -> (Custom Idle is Playing:double)
   ) {
    curveBlendOption=NormalizeByWeight
    layerSetup [{branchFilters [
       {boneName=C_brow_mid}
       {boneName=C_lip_upper_mid}
       {boneName=C_nose_bridge}
       {boneName=L_brow_mid}
       {boneName=L_brow_outer}
       {boneName=L_cheek_inner}
       {boneName=L_eye}
       {boneName=L_lip_corner}
       {boneName=L_lip_upper_outer}
       {boneName=R_brow_mid}
       {boneName=R_brow_outer}
       {boneName=R_cheek_inner}
       {boneName=R_eye}
       {boneName=R_lip_corner}
       {boneName=R_lip_upper_outer}
      ]}]
   }
   Output Pose(Result=@n1) {layerGroup=None}
   n2 = Control Rig!(Source=Copy Pose From Mesh() {bUseAttachedParent=true}, bAlphaBoolEnabled:bool=True) -> (Pose) {
    alphaInputType=Bool
    alphaBoolBlend {blendOption=Linear}
    controlRigAssetReference {blueprintRigClass=&/M_MED_LocalZilla_Head_CtrlRig_Tongue.M_MED_LocalZilla_Head_CtrlRig_Tongue_C}
   }
   n3 = Get headband dynamics alpha() -> (headband dynamics alpha:bool)
   n4 = Anim Dynamics - Bone: dyn_headband_front!(ComponentPose=Local To Component(LocalPose=@n2), bAlphaBoolEnabled:bool=@n3, ExternalForce="0,0,0") -> (Pose) {
    alphaInputType=Bool
    angularDampingOverride=1
    angularSpringConstant=20000
    bAngularSpring=true
    bOverrideAngularDamping=true
    bOverrideLinearDamping=true
    bUseGravityOverride=true
    gravityScale=0.35
    linearDampingOverride=1
    linearSpringConstant=10
    simulationSpace=BoneRelative
    alphaBoolBlend {blendOption=Linear}
    boundBone {boneName=dyn_headband_front}
    componentLinearAccScale=<1 1 1>
    componentLinearVelScale=<1 1 1>
    gravityOverride {z=-980}
    physicsBodyDefinitions [{
      boundBone {boneName=dyn_headband_front}
      boxExtents=<50 50 50>
      constraintSetup {angularLimitsMax=<20 20 20>, angularLimitsMin=<-20 -20 -20>, angularTarget=<1 0 0>, linearAxesMax=<0.5 1 0>, linearAxesMin=<-0.5 -1 0>}
      localJointOffset=<0 1 0>
     }]
    relativeSpaceBone {boneName=spine_03}
   }
   n5 = RigidBody!(ComponentPose=@n4, bAlphaBoolEnabled:bool=@n3, ExternalForce) -> (Pose) {
    alphaInputType=Bool
    bClampLinearTranslationLimitToRefPose=true
    bOverrideWorldGravity=true
    overridePhysicsAsset=&/M_MED_LocalZilla_Head_Physics
    simulationSpace=BaseBoneSpace
    alphaBoolBlend {blendOption=Linear}
    baseBoneRef {boneName=spine_04}
    componentLinearAccScale=<0.2 0.2 0.2>
    componentLinearVelScale=<0.22 0.22 0.22>
    overrideWorldGravity {z=-980}
   }
   Save cached pose 'Final Pose'(Pose=Component To Local(ComponentPose=@n5)) {cacheName=Final Pose}
   Get turnOffTongue() -> (turnOffTongue:bool)
  ]}
 Calculate Cloth Sim Noise {nodes [
   n1 = Calculate Cloth Sim Noise!() -> (
    then
    MID:Material Instance Dynamic Object Reference
    OverlayMID:Material Instance Dynamic Object Reference
    SkelMesh:Skeletal Mesh Component Object Reference
    BoneName:Name=root
    DeltaTime:double=~
    NoiseMovementStrength:double=0.25
    NoiseBoneLocation:"Vector (by ref)"
    NoiseVelocityVector:"Vector (by ref)"
   ) {is=title K2Node_FunctionEntry}
   > n2 = calculateMotionVectorForSocket!(
    execute
    self
    skelMesh:Scene Component Object Reference=@n1.SkelMesh
    boneName:Name=@n1.BoneName
    previousBoneLocation:Vector=@n1.NoiseBoneLocation
   ) -> (then, boneMotionVector:Vector, boneLocation:Vector, boneTransform:Transform)
   n3 = vector + vector(
    A=vector * vector(A=@n2, B=float * float(A=@n1.DeltaTime, B=@n1.NoiseMovementStrength) {is=title K2Node_PromotableOperator}) -> (ReturnValue=~)
    B=@n1.NoiseVelocityVector
   )
   > SetVectorParameterValue(
    self=[@n1.OverlayMID, @n1]
    ParameterName=ClothSim_ABP_TurbulenceUVOffset
    Value="To LinearColor (Vector)"!(InVec:Vector=@n3) -> (ReturnValue:Linear Color Structure)
   )
   > "Set Vector (by ref)"!(execute:Exec, Target:"Vector (by ref)"=@n1.NoiseBoneLocation, Value:Vector=@n2.boneLocation) -> (then:Exec) {is=title K2Node_VariableSetRef}
   > "Set Vector (by ref)"(Target=@n1.NoiseVelocityVector, Value=@n3)
   > Return Node()
  ]}
 EventGraph {nodes [
   n1 = Event BlueprintUpdateAnimation()
   > n2 = Sequence!(execute) -> (then_0, then_1, then_2:Exec, then_3:Exec, then_4:Exec)
   > n3 = Set skeletal_mesh(skeletal_mesh:Skeletal Mesh Component Object Reference=GetPartSkeletalMeshComponent()) -> (Output_Get:Skeletal Mesh Component Object Reference)
   > n4 = Get Dynamic Material Instance(SkeletalMeshComponent=@n3)
   > Set Mat Instance Dynamic(Mat Instance Dynamic=@n4)
   > n5 = GetOrCreateDynamicOverlayMaterial(MeshComponent=@n3, MaterialName=None)
   > Set Overlay Mat Instance Dynamic(Overlay Mat Instance Dynamic=@n5)
   n6 = Get bIsSkyDiving()
   n7 = Branch(execute=@n2.then_1, Condition=@n6)
   > n8 = Set ClothSim_WindStrength(ClothSim_WindStrength:double=1) -> (Output_Get:double)
   n9 = Get Mat Instance Dynamic()
   SetScalarParameterValue(
    execute=[@n10.then, @n8.then]
    self=@n9
    ParameterName=ClothSim_WindStrength
    Value=Get ClothSim_WindStrength() -> (ClothSim_WindStrength:double)
   )
   n10 = Set ClothSim_WindStrength(execute=@n7.else, ClothSim_WindStrength=0)
   n11 = Branch(execute=@n2.then_2, Condition=@n6)
   > n12 = Set ClothSim_VelocityMax(ClothSim_VelocityMax:double=750) -> (Output_Get:double)
   SetScalarParameterValue(
    execute=[@n13.then, @n12.then]
    self=@n9
    ParameterName=ClothSim_VelocityMax
    Value=Get ClothSim_VelocityMax() -> (ClothSim_VelocityMax:double)
   )
   n13 = Set ClothSim_VelocityMax(execute=@n11.else, ClothSim_VelocityMax=1500)
   Set WPO Globals Function(
    execute=@n2.then_3
    Material=Get Mat Instance Dynamic()
    Overlay Material=Get Overlay Mat Instance Dynamic()
    Skeletal Mesh=Get skeletal_mesh() -> (skeletal_mesh:Skeletal Mesh Component Object Reference)
    "Local Noise Transform  Location"=Get Local Noise Transform Location() -> (Local Noise Transform Location:Vector)
    Local Noise Offset=Get Local Noise Offset()
   )
   > Set WPO Regional Function!(
    execute
    Material:Material Instance Dynamic Object Reference=Get Mat Instance Dynamic()
    Overlay Material:Material Instance Dynamic Object Reference=Get Overlay Mat Instance Dynamic()
    Skeletal Mesh:Skeletal Mesh Component Object Reference=Get skeletal_mesh()
    Velocity Transform:Name=neck_01
    Velocity Interp Speed:double=Get Velocity Interp Speed()
    Collision Transform:Name=calf_r
    Collision Transform Axis:EAxis Enum
    Invert Collision Transform Axis:bool=false
    Delta Time X:double=@n1.DeltaTimeX
    Previous Velocity Transform Location:"Vector (by ref)"=Get R1 Velocity Transform Location()
    Previous Velocity Transform Velocity:"Vector (by ref)"=Get R1 Velocity Transform Velocity()
    Region:int=1
   ) -> (then)
   Branch(execute=@n2.then_4, Condition=Get bIsInFrontEnd())
   > Set Custom Idle is Playing(Custom Idle is Playing:double=GetCurveValue(self=GetOwnerFrontEndAnimInstance(), CurveName=Frontend_Idle_Bypass)) -> (Output_Get:double)
   Event OnCharacterCustomizationLoaded()
   > n14 = Branch(Condition=CharacterItemDefinitionHasAnyMetaTag(Tags=Make Array("[0]"=(TagName="Cosmetics.Variant.Property.Stage2"))))
   > Set headband dynamics alpha(headband dynamics alpha:bool=true) -> (Output_Get:bool)
   Set headband dynamics alpha(execute=@n14.else, headband dynamics alpha=false)
   Event BlueprintInitializeAnimation()
   n15 = GetOwnerSkeletalMeshComponent()
   > Is Valid(InputObject=@n15)
   > Set Skeletal_Mesh(Skeletal_Mesh:Skeletal Mesh Component Object Reference=@n15) -> (Output_Get:Skeletal Mesh Component Object Reference)
   > n16 = Cast To FortPlayerPawn(Object=GetOwningActor())
   > n17 = Set Pawn(Pawn:Fort Player Pawn Object Reference=@n16) -> (Output_Get:Fort Player Pawn Object Reference)
   > Set Anim_Inst(Anim_Inst:Anim Instance Object Reference=GetAnimInstance(self=Get Mesh(self=@n17))) -> (Output_Get:Anim Instance Object Reference)
   Is Valid()
   n18 = Get bIsPlayingEmote()
   > n19 = Branch(Condition=AND Boolean(
     A=Montage_IsPlaying(Montage=/BRCosmetics/Animation/Game/MainPlayer/Menu/BR/CustomSkins/_s24/LocalZillaM/LocalZilla_Male_Idle_Montage)
     B=NOT Boolean(A=@n18)
    ))
   > Set turnOffTongue(turnOffTongue:bool=false) -> (Output_Get:bool)
   Set turnOffTongue(execute=@n19.else, turnOffTongue=true)
   Reroute Node(InputPin:Wildcard) -> (OutputPin:Wildcard)
  ]}
 GetBoneNames {nodes [
   n1 = GetBoneNames!() -> (then, Region:"Integer (by ref)") {is=title K2Node_FunctionEntry}
   n2 = Append(A=ClothSim_ABP_R, B="To String (Integer)"(InInt=@n1))
   > Return Node!(
    execute
    Velocity:Name=String To Name(InString=Append(A=@n2, B=_Velocity))
    BonePosition:Name=String To Name(InString=Append(A=@n2, B=_BonePosition))
    BoneVector:Name=String To Name(InString=Append(A=@n2, B=_BoneVector))
   )
  ]}
 Set Cloth Sim Params {nodes [
   n1 = Set Cloth Sim Params!() -> (
    then
    material:Material Instance Dynamic Object Reference
    overlayMaterial:Material Instance Dynamic Object Reference
    skelMesh:Skeletal Mesh Component Object Reference
    boneName:Name
    interpSpeed:double=2
    collisionBoneName:Name
    collisionBoneAxis:EAxis Enum=None
    invertCollisionBoneAxis:bool=~
    deltaTimeX:double=~
    previousBoneLocation:"Vector (by ref)"
    previousBoneVelocityVector:"Vector (by ref)"
    region:int=1
   ) {is=title K2Node_FunctionEntry}
   > Set lMID(lMID:Material Instance Dynamic Object Reference=@n1) -> (Output_Get:Material Instance Dynamic Object Reference)
   > Set loverlayMID(loverlayMID:Material Instance Dynamic Object Reference=@n1.overlayMaterial) -> (Output_Get:Material Instance Dynamic Object Reference)
   > Set lSkelMesh(lSkelMesh:Skeletal Mesh Component Object Reference=@n1.skelMesh) -> (Output_Get:Skeletal Mesh Component Object Reference)
   > n2 = GetBoneNames!(execute, self, Region:"Integer (by ref)"=@n1.region) -> (then, Velocity:Name=None, BonePosition:Name=None, BoneVector:Name=None) {is=title K2Node_CallFunction}
   > Set lMatVelocityParamName(lMatVelocityParamName:Name=@n2) -> (Output_Get:Name=None)
   > Set lMatBonePositionParamName(lMatBonePositionParamName:Name=@n2.BonePosition) -> (Output_Get:Name=None)
   > Set lMatBoneVectorParamName(lMatBoneVectorParamName:Name=@n2.BoneVector) -> (Output_Get:Name=None)
   > n3 = calculateMotionVectorForSocket(
    skelMesh=Get lSkelMesh() -> (lSkelMesh:Skeletal Mesh Component Object Reference)
    boneName=@n1.boneName
    previousBoneLocation=@n1.previousBoneLocation
   )
   n4 = VInterpTo!(
    Current:Vector=@n1.previousBoneVelocityVector
    Target:Vector=vector / vector(A:Vector=@n3, B=@n1.deltaTimeX) -> (ReturnValue:Vector)
    DeltaTime:float=@n1.deltaTimeX
    InterpSpeed:float=@n1.interpSpeed
   ) -> (ReturnValue:Vector)
   > SetVectorParameterValue(
    self=[
     Get loverlayMID() -> (loverlayMID:Material Instance Dynamic Object Reference)
     Get lMID() -> (lMID:Material Instance Dynamic Object Reference)
    ]
    ParameterName=Get lMatVelocityParamName() -> (lMatVelocityParamName:Name=None)
    Value="To LinearColor (Vector)"(InVec=@n4)
   )
   n5 = GetSocketTransform(self=Get lSkelMesh(), InSocketName=@n1.collisionBoneName, TransformSpace=RTS_World)
   n6 = "Get Unit Axis (Matrix)"!(
    M:"Matrix Structure (by ref)"="To Matrix (Transform)"!(Transform:"Transform (by ref)"=@n5) -> (ReturnValue:Matrix Structure)
    Axis:EAxis Enum=@n1.collisionBoneAxis
   ) -> (ReturnValue:Vector) {is=title K2Node_CallFunction}
   n7 = Select(Option 0:double=1, Option 1:double=-1, Index=@n1.invertCollisionBoneAxis) -> (ReturnValue:double)
   > SetVectorParameterValue(
    self=[Get loverlayMID(), Get lMID()]
    ParameterName=Get lMatBoneVectorParamName() -> (lMatBoneVectorParamName:Name=None)
    Value="To LinearColor (Vector)"(InVec=vector * vector(A=@n6, B=@n7) -> (ReturnValue=~))
   )
   > SetVectorParameterValue(
    self=[Get loverlayMID(), Get lMID()]
    ParameterName=Get lMatBonePositionParamName() -> (lMatBonePositionParamName:Name=None)
    Value="To LinearColor (Vector)"(InVec=BreakTransform(InTransform=@n5))
   )
   > "Set Vector (by ref)"(Target=@n1.previousBoneLocation, Value=@n3.boneLocation)
   > "Set Vector (by ref)"(Target=@n1.previousBoneVelocityVector, Value=@n4)
   > Return Node!(execute)
   Set lMID()
  ]}
 calculateMotionVectorForSocket {nodes [
   n1 = calculateMotionVectorForSocket!() -> (then, skelMesh:Scene Component Object Reference, boneName:Name, previousBoneLocation:Vector=~) {is=title K2Node_FunctionEntry}
   n2 = GetSocketTransform(self=@n1, InSocketName=@n1.boneName, TransformSpace=RTS_World)
   n3 = BreakTransform(InTransform=@n2)
   > Return Node!(execute, boneMotionVector:Vector=vector - vector(A=@n3, B=@n1.previousBoneLocation), boneLocation:Vector=@n3, boneTransform:Transform=@n2)
  ]}
 setWPOMaterialVelocity {nodes [
   n1 = setWPOMaterialVelocity!() -> (
    then
    materialInstance:Material Instance Dynamic Object Reference
    skelMesh:Skeletal Mesh Component Object Reference
    boneName:Name
    offsetBoneName:Name
    offsetBoneAxis:EAxis Enum=None
    deltaTimeX:double=~
    materialVelocityAttribute:Name=Velocity
    previousBoneLocation:Vector=~
    previousBoneVelocityVector:Vector=~
   ) {is=title K2Node_FunctionEntry}
   > n2 = calculateMotionVectorForSocket(skelMesh=@n1.skelMesh, boneName=@n1.boneName, previousBoneLocation=@n1.previousBoneLocation) {is=title K2Node_CallFunction}
   n3 = "Get Unit Axis (Matrix)"(
    M="To Matrix (Transform)"(Transform=GetSocketTransform(self=@n1.skelMesh, InSocketName=@n1.offsetBoneName, TransformSpace=RTS_World))
    Axis=@n1.offsetBoneAxis
   )
   n4 = vector / vector(A=@n2, B=@n1.deltaTimeX)
   n5 = VInterpTo(Current=@n1.previousBoneVelocityVector, Target=@n4, DeltaTime=@n1.deltaTimeX, InterpSpeed=0.75)
   > SetVectorParameterValue(self=@n1, ParameterName=@n1.materialVelocityAttribute, Value="To LinearColor (Vector)"(InVec=vector + vector(A=@n3, B=@n5)))
   > Return Node!(
    execute
    boneVelocity:double=Vector Length(A=@n4)
    boneVelocityVector:Vector=@n4
    boneLocation=@n2.boneLocation
    boneTransform=@n2.boneTransform
   )
  ]}
]
variables [
 Anim_Inst {type=AnimInstance, default=null}
 ClothSim_VelocityMax {type=number}
 ClothSim_WindStrength {type=number}
 Custom Idle is Playing {type=number}
 headband dynamics alpha {type=boolean}
 Local Noise Offset {type=Vector}
 Local Noise Transform Location {type=Vector}
 Mat Instance Dynamic {type=MaterialInstanceDynamic}
 Overlay Mat Instance Dynamic {type=MaterialInstanceDynamic}
 Pawn {type=FortPlayerPawn}
 R1 Velocity Transform Location {type=Vector}
 R1 Velocity Transform Velocity {type=Vector}
 skeletal_mesh {type=SkeletalMeshComponent}
 turnOffTongue {type=boolean, default=true}
 Velocity Interp Speed {type=number}
]
